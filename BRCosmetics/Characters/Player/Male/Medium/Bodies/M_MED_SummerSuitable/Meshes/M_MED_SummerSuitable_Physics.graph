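# Physics asset for the SummerSuitable male medium body, covering a pouch with two straps and the right thigh it is attached to.
# Capsules approximate the thigh, its twist bone and the pouch, with a tiny capsule on the pouch root.
# Angularly locked constraints tie the pouch bones to each other and to the thigh.
# Each strap is a two-link chain off the pouch with limited swing and zero twist, while the thigh twist joint rotates freely.
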PhysicsAsset /BRCosmetics/Characters/Player/Male/Medium/Bodies/M_MED_SummerSuitable/Meshes/M_MED_SummerSuitable_Physics
bodies {
 dyn_pouch_a [{shapeType=Capsule, length=0.505, radius=0.505, shapeName=dyn_pouch_a_capsule, rotation=<rot 0 -90 0>}]
 dyn_pouch_b [
  {shapeType=Capsule, length=7, radius=5.5, shapeName=dyn_pouch_b_capsule, center=<14.6818 0 -2.6188>, rotation=<rot 75.9782 0 90.0001>}
  {shapeType=Capsule, length=7, radius=4.7, shapeName=dyn_pouch_b_capsule, center=<5.6555 0 -0.5605>, rotation=<rot 76.7145 0 90.0001>}
 ]
 dyn_pouch_strap_aa [{shapeType=Capsule, length=4, radius=1, shapeName=dyn_pouch_strap_aa_capsule, center=<5 0 0>, rotation=<rot 90 0 0>}]
 dyn_pouch_strap_ab [{shapeType=Capsule, length=3, radius=1, shapeName=dyn_pouch_strap_ab_capsule, center=<4 0 0>, rotation=<rot 90 0 0>}]
 dyn_pouch_strap_ba [{shapeType=Capsule, length=4, radius=1, shapeName=dyn_pouch_strap_ba_capsule, center=<5 0 0>, rotation=<rot 90 0 0>}]
 dyn_pouch_strap_bb [{shapeType=Capsule, length=3, radius=1, shapeName=dyn_pouch_strap_bb_capsule, center=<4 0 0>, rotation=<rot 90 0 0>}]
 thigh_r [{shapeType=Capsule, length=40, radius=10, shapeName=thigh_r_capsule, center=<12.6393 0 -1.9661>, rotation=<rot 90 0 0>}]
 thigh_twist_01_r [{
   shapeType=Capsule
   length=20
   radius=10
   shapeName=thigh_twist_01_r_capsule
   center=<-8.8205 -1.1899 -1.9932>
   rotation=<rot 0.0382 -88.3015 -89.6976>
  }]
}
constraints [
 {bone1Name=dyn_pouch_strap_aa, bone2Name=dyn_pouch_b, swing1LimitDegrees=30, swing2LimitDegrees=60, twistLimitDegrees=0}
 {bone1Name=dyn_pouch_strap_ab, bone2Name=dyn_pouch_strap_aa, swing1LimitDegrees=30, swing2LimitDegrees=40, twistLimitDegrees=0}
 {bone1Name=dyn_pouch_strap_ba, bone2Name=dyn_pouch_b, swing1LimitDegrees=30, swing2LimitDegrees=60, twistLimitDegrees=0}
 {bone1Name=dyn_pouch_strap_bb, bone2Name=dyn_pouch_strap_ba, swing1LimitDegrees=30, swing2LimitDegrees=40, twistLimitDegrees=0}
 {
  bone1Name=dyn_pouch_b
  bone2Name=dyn_pouch_a
  swing1LimitDegrees=20
  swing1Motion=Locked
  swing2LimitDegrees=30
  swing2Motion=Locked
  twistLimitDegrees=0
  twistMotion=Locked
 }
 {
  bone1Name=thigh_r
  bone2Name=dyn_pouch_a
  swing1LimitDegrees=45
  swing1Motion=Locked
  swing2LimitDegrees=45
  swing2Motion=Locked
  twistLimitDegrees=45
  twistMotion=Locked
 }
 {
  bone1Name=thigh_twist_01_r
  bone2Name=thigh_r
  swing1LimitDegrees=45
  swing1Motion=Free
  swing2LimitDegrees=45
  swing2Motion=Free
  twistLimitDegrees=45
  twistMotion=Free
 }
]
