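# Animation Blueprint for the male medium SummerSuitable cosmetic body, adding secondary motion to straps, pouches, zipper and pants.
# The parent mesh pose passes through Control Rig, weighted constraints, look-ats, RigidBody, AnimDynamics, then stacked additive bone offsets.
# The event graph derives thigh yaw to scale strap and buckle correctives; a skydive-dive flag gates one pouch offset.
# It also computes gravity override, knee and movement-state values that the visible AnimGraph never reads.
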AnimBlueprint /BRCosmetics/Characters/Player/Male/Medium/Bodies/M_MED_SummerSuitable/Meshes/M_MED_SummerSuitable_AnimBP
target_skeleton=&/M_MED_SummerSuitable_Skeleton
graphs [
 AnimGraph {nodes [
   "Transform (Modify) Bone - Bone: dyn_pants_base_l"(Alpha=1, Translation="-2,0,0", Rotation="0,0,0", Scale="1,1,1") {
    rotationMode=BMM_Additive
    rotationSpace=BCS_BoneSpace
    translationMode=BMM_Additive
    translationSpace=BCS_BoneSpace
    alphaBoolBlend {blendOption=Linear}
    alphaScaleBiasClamp {bInterpResult=true, interpSpeedDecreasing=5}
    boneToModify {boneName=dyn_pants_base_l}
    translation {x=-2}
   }
   "Transform (Modify) Bone - Bone: dyn_pouch_b"!(ComponentPose, bAlphaBoolEnabled:bool=True, Translation="0,0,0", Rotation="7,0,8", Scale="1,1,1") -> (Pose) {
    alphaInputType=Bool
    rotationMode=BMM_Additive
    translationMode=BMM_Additive
    alphaBoolBlend {blendOption=Linear}
    alphaScaleBiasClamp {bInterpResult=true, interpSpeedIncreasing=2}
    boneToModify {boneName=dyn_pouch_b}
    rotation {pitch=7, roll=8}
   }
   "Transform (Modify) Bone - Bone: dyn_pouch_b"(bAlphaBoolEnabled=True, Translation="0,0,0", Rotation="7,0,8", Scale="1,1,1") {
    alphaInputType=Bool
    rotationMode=BMM_Additive
    translationMode=BMM_Additive
    alphaBoolBlend {blendOption=Linear}
    alphaScaleBiasClamp {bInterpResult=true, interpSpeedIncreasing=2}
    boneToModify {boneName=dyn_pouch_b}
    rotation {pitch=7, roll=8}
   }
   n1 = Control Rig(Source=Copy Pose From Mesh() {bUseAttachedParent=true}) {
    alphaBoolBlend {blendOption=Linear}
    controlRigAssetReference {blueprintRigClass=&/M_MED_SummerSuitable_CtrlRig.M_MED_SummerSuitable_CtrlRig_C}
   }
   n2 = "Constraint - dyn_fr_strap_base_l (PPPPP)"!(
    ComponentPose=Local To Component(LocalPose=@n1)
    Alpha=1
    ConstraintWeights_0=0.4764
    ConstraintWeights_1:float=0.2383
    ConstraintWeights_2:float=0.1879
    ConstraintWeights_3:float=0.0633
    ConstraintWeights_4:float=0.0341
   ) -> (Pose) {
    alphaBoolBlend {blendOption=Linear}
    boneToModify {boneName=dyn_fr_strap_base_l}
    constraintSetup [
     {targetBone {boneName=pelvis}}
     {targetBone {boneName=deform_groin_l}}
     {targetBone {boneName=thigh_twist_01_l}}
     {targetBone {boneName=thigh_twist_01_r}}
     {targetBone {boneName=deform_groin_r}}
    ]
    constraintWeights [1,1,1,1,1]
   }
   n3 = "Constraint - dyn_fr_strap_base_r (PPPPP)"!(
    ComponentPose=@n2
    Alpha=1
    ConstraintWeights_0=0.4698
    ConstraintWeights_1:float=0.2401
    ConstraintWeights_2:float=0.1928
    ConstraintWeights_3:float=0.0635
    ConstraintWeights_4:float=0.0337
   ) -> (Pose) {
    alphaBoolBlend {blendOption=Linear}
    boneToModify {boneName=dyn_fr_strap_base_r}
    constraintSetup [
     {targetBone {boneName=pelvis}}
     {targetBone {boneName=deform_groin_r}}
     {targetBone {boneName=thigh_twist_01_r}}
     {targetBone {boneName=thigh_twist_01_l}}
     {targetBone {boneName=deform_groin_l}}
    ]
    constraintWeights [1,1,1,1,1]
   }
   n4 = "Constraint - dyn_fr_strap_base_l (R)"(ComponentPose=@n3, Alpha=1, ConstraintWeights_0=1) {
    alphaBoolBlend {blendOption=Linear}
    boneToModify {boneName=dyn_fr_strap_base_l}
    constraintSetup [{transformType=Rotation, targetBone {boneName=thigh_l}}]
    constraintWeights [1]
   }
   n5 = "Constraint - dyn_fr_strap_base_r (R)"(ComponentPose=@n4, Alpha=1, ConstraintWeights_0=1) {
    alphaBoolBlend {blendOption=Linear}
    boneToModify {boneName=dyn_fr_strap_base_r}
    constraintSetup [{transformType=Rotation, targetBone {boneName=thigh_r}}]
    constraintWeights [1]
   }
   n6 = "Transform (Modify) Bone - Bone: dyn_fr_strap_base_l"(ComponentPose=@n5, Alpha=Get uplegRot_L() -> (uplegRot_L:double), Translation="1,1,0", Rotation="0,15,0.000", Scale="1,1,1") {
    rotationMode=BMM_Additive
    rotationSpace=BCS_BoneSpace
    translationMode=BMM_Additive
    translationSpace=BCS_BoneSpace
    alphaBoolBlend {blendOption=Linear}
    alphaScaleBiasClamp {bMapRange=true, inRange {max=0, min=-100}, outRange {max=0, min=1}}
    boneToModify {boneName=dyn_fr_strap_base_l}
    rotation {yaw=15}
    translation=<1 1>
   }
   n7 = "Transform (Modify) Bone - Bone: dyn_fr_strap_base_r"(ComponentPose=@n6, Alpha=Get uplegRot_R() -> (uplegRot_R:double), Translation="0,0,0", Rotation="0,20,0", Scale="1,1,1") {
    rotationMode=BMM_Additive
    rotationSpace=BCS_BoneSpace
    translationMode=BMM_Additive
    translationSpace=BCS_BoneSpace
    alphaBoolBlend {blendOption=Linear}
    alphaScaleBiasClamp {bMapRange=true, inRange {max=0, min=-100}, outRange {max=0, min=1}}
    boneToModify {boneName=dyn_fr_strap_base_r}
    rotation {yaw=20}
   }
   n8 = "Constraint - dyn_pouch_a_base (PPPPPP)"!(
    ComponentPose=@n7
    Alpha=1
    ConstraintWeights_0=0.8241
    ConstraintWeights_1:float=0.0506
    ConstraintWeights_2:float=0.0489
    ConstraintWeights_3:float=0.0369
    ConstraintWeights_4:float=0.0264
    ConstraintWeights_5:float=0.0131
   ) -> (Pose) {
    alphaBoolBlend {blendOption=Linear}
    boneToModify {boneName=dyn_pouch_a_base}
    constraintSetup [
     {targetBone {boneName=pelvis}}
     {targetBone {boneName=spine_01}}
     {targetBone {boneName=spine_02}}
     {targetBone {boneName=thigh_twist_01_r}}
     {targetBone {boneName=spine_03}}
     {targetBone {boneName=deform_groin_r}}
    ]
    constraintWeights [1,1,1,1,1,1]
   }
   n9 = "Constraint - dyn_pouch_bukle_strap_a_tgt (PPPPP)"!(
    ComponentPose=@n8
    Alpha=1
    ConstraintWeights_0=0.7823
    ConstraintWeights_1:float=0.0855
    ConstraintWeights_2:float=0.0762
    ConstraintWeights_3:float=0.0357
    ConstraintWeights_4:float=0.0204
   ) -> (Pose) {
    alphaBoolBlend {blendOption=Linear}
    boneToModify {boneName=dyn_pouch_bukle_strap_a_tgt}
    constraintSetup [
     {targetBone {boneName=pelvis}}
     {targetBone {boneName=thigh_twist_01_r}}
     {targetBone {boneName=deform_groin_r}}
     {targetBone {boneName=spine_02}}
     {targetBone {boneName=spine_01}}
    ]
    constraintWeights [1,1,1,1,1]
   }
   n10 = "Constraint - dyn_pouch_bukle_strap_b_tgt (PPPPPP)"!(
    ComponentPose=@n9
    Alpha=1
    ConstraintWeights_0=0.8745
    ConstraintWeights_1:float=0.0403
    ConstraintWeights_2:float=0.0348
    ConstraintWeights_3:float=0.0228
    ConstraintWeights_4:float=0.0183
    ConstraintWeights_5:float=0.0094
   ) -> (Pose) {
    alphaBoolBlend {blendOption=Linear}
    boneToModify {boneName=dyn_pouch_bukle_strap_b_tgt}
    constraintSetup [
     {targetBone {boneName=pelvis}}
     {targetBone {boneName=spine_02}}
     {targetBone {boneName=spine_01}}
     {targetBone {boneName=spine_03}}
     {targetBone {boneName=thigh_twist_01_r}}
     {targetBone {boneName=deform_glute_r}}
    ]
    constraintWeights [1,1,1,1,1,1]
   }
   n11 = Look At - Bone: dyn_pouch_bukle_strap_a_aim(ComponentPose=@n10, Alpha=1) {
    bUseLookUpAxis=true
    alphaBoolBlend {blendOption=Linear}
    boneToModify {boneName=dyn_pouch_bukle_strap_a_aim}
    lookAtLocation {x=0}
    lookAtTarget {boneReference {boneName=dyn_pouch_bukle_strap_a_tgt}}
    lookAt_Axis {axis=<1 0>}
    lookUp_Axis {axis {x=0, z=1}}
   }
   n12 = Look At - Bone: dyn_pouch_bukle_strap_b_aim(ComponentPose=@n11, Alpha=1) {
    bUseLookUpAxis=true
    alphaBoolBlend {blendOption=Linear}
    boneToModify {boneName=dyn_pouch_bukle_strap_b_aim}
    lookAtLocation {x=0}
    lookAtTarget {boneReference {boneName=dyn_pouch_bukle_strap_b_tgt}}
    lookAt_Axis {axis=<1 0>}
    lookUp_Axis {axis {x=0, z=1}}
   }
   n13 = "Transform (Modify) Bone - Bone: dyn_pouch_bukle_strap_a_aim"(ComponentPose=@n12, Alpha=Get uplegRot_R(), Translation="0,0,0", Rotation="40,0,0", Scale="1,1,1") {
    rotationMode=BMM_Additive
    rotationSpace=BCS_BoneSpace
    translationMode=BMM_Additive
    translationSpace=BCS_BoneSpace
    alphaBoolBlend {blendOption=Linear}
    alphaScaleBiasClamp {bMapRange=true, inRange {max=-70, min=-100}, outRange {max=0, min=1}}
    boneToModify {boneName=dyn_pouch_bukle_strap_a_aim}
    rotation {pitch=40}
   }
   n14 = RigidBody!(ComponentPose=@n13, Alpha, OverrideWorldGravity:Vector="0,0,-980", ExternalForce) -> (Pose) {
    bClampLinearTranslationLimitToRefPose=true
    bOverrideWorldGravity=true
    overridePhysicsAsset=&/M_MED_SummerSuitable_Physics
    simulationSpace=BaseBoneSpace
    alphaBoolBlend {blendOption=Linear}
    baseBoneRef {boneName=pelvis}
    componentAppliedLinearAccClamp=<2000 2000 2000>
    overrideWorldGravity {z=-980}
    simSpaceSettings {maxAngularAcceleration=100, worldAlpha=1}
   }
   n15 = Anim Dynamics - Bone: dyn_pouch_a!(ComponentPose=@n14, Alpha=0.7, GravityOverride:Vector="0,0,-980", ExternalForce="0,0,0") -> (Pose) {
    angularDampingOverride=0.9999
    angularSpringConstant=5000
    bAngularSpring=true
    bOverrideAngularDamping=true
    bUseGravityOverride=true
    gravityScale=0.35
    simulationSpace=RootRelative
    alphaBoolBlend {blendOption=Linear}
    boundBone {boneName=dyn_pouch_a}
    componentAppliedLinearAccClamp=<1000 1000 1000>
    componentLinearAccScale=<1 1 1>
    componentLinearVelScale=<1 1 1>
    gravityOverride {z=-980}
    physicsBodyDefinitions [{
      boundBone {boneName=dyn_pouch_a}
      boxExtents=<20 20 20>
      constraintSetup {angularLimitsMax=<0 0 10>, angularLimitsMin=<0 -15 -10>, angularTarget=<1 0 0>}
      localJointOffset=<2 0 0>
     }]
   }
   n16 = Anim Dynamics - Bone: dyn_pouch_b!(ComponentPose=@n15, Alpha=0.7, GravityOverride:Vector="0,0,-980", ExternalForce="0,0,0") -> (Pose) {
    angularDampingOverride=0.9999
    angularSpringConstant=5000
    bAngularSpring=true
    bOverrideAngularDamping=true
    bUseGravityOverride=true
    gravityScale=0.35
    simulationSpace=RootRelative
    alphaBoolBlend {blendOption=Linear}
    boundBone {boneName=dyn_pouch_b}
    componentAppliedLinearAccClamp=<1000 1000 1000>
    componentLinearAccScale=<1 1 1>
    componentLinearVelScale=<1 1 1>
    gravityOverride {z=-980}
    physicsBodyDefinitions [{
      boundBone {boneName=dyn_pouch_b}
      boxExtents=<20 20 20>
      constraintSetup {angularLimitsMax=<0 0 10>, angularLimitsMin=<0 -20 -10>, angularTarget=<1 0 0>}
      localJointOffset=<2 0 0>
     }]
   }
   n17 = Anim Dynamics - Bone: dyn_fr_strap_r!(ComponentPose=@n16, Alpha=1, GravityOverride:Vector="0,0,-980", ExternalForce="0,0,0") -> (Pose) {
    angularDampingOverride=1
    angularSpringConstant=1500
    bAngularSpring=true
    bOverrideAngularDamping=true
    bUseGravityOverride=true
    gravityScale=0.35
    simulationSpace=RootRelative
    alphaBoolBlend {blendOption=Linear}
    boundBone {boneName=dyn_fr_strap_r}
    componentAppliedLinearAccClamp=<2000 2000 2000>
    componentLinearAccScale=<1 1 1>
    componentLinearVelScale=<1 1 1>
    gravityOverride {z=-980}
    physicsBodyDefinitions [{
      boundBone {boneName=dyn_fr_strap_r}
      boxExtents=<10 10 10>
      constraintSetup {angularLimitsMax=<0 0 25>, angularLimitsMin=<0 -20 -25>, angularTarget=<1 0 0>}
      localJointOffset=<1.5 0 0>
     }]
   }
   n18 = Anim Dynamics - Bone: dyn_fr_strap_l!(ComponentPose=@n17, Alpha=1, GravityOverride:Vector="0,0,-980", ExternalForce="0,0,0") -> (Pose) {
    angularDampingOverride=1
    angularSpringConstant=1500
    bAngularSpring=true
    bOverrideAngularDamping=true
    bUseGravityOverride=true
    gravityScale=0.35
    simulationSpace=RootRelative
    alphaBoolBlend {blendOption=Linear}
    boundBone {boneName=dyn_fr_strap_l}
    componentAppliedLinearAccClamp=<2000 2000 2000>
    componentLinearAccScale=<1 1 1>
    componentLinearVelScale=<1 1 1>
    gravityOverride {z=-980}
    physicsBodyDefinitions [{
      boundBone {boneName=dyn_fr_strap_l}
      boxExtents=<10 10 10>
      constraintSetup {angularLimitsMax=<0 0 25>, angularLimitsMin=<0 -20 -25>, angularTarget=<1 0 0>}
      localJointOffset=<1.5 0 0>
     }]
   }
   n19 = "Constraint - dyn_fr_zip (PPPPPP)"!(
    ComponentPose=@n18
    Alpha=1
    ConstraintWeights_0=0.4383
    ConstraintWeights_1:float=0.2132
    ConstraintWeights_2:float=0.155
    ConstraintWeights_3:float=0.0675
    ConstraintWeights_4:float=0.0634
    ConstraintWeights_5:float=0.0626
   ) -> (Pose) {
    alphaBoolBlend {blendOption=Linear}
    boneToModify {boneName=dyn_fr_zip}
    constraintSetup [
     {targetBone {boneName=spine_04}}
     {targetBone {boneName=spine_05}}
     {targetBone {boneName=spine_03}}
     {targetBone {boneName=clavicle_l}}
     {targetBone {boneName=spine_02}}
     {targetBone {boneName=clavicle_r}}
    ]
    constraintWeights [1,1,1,1,1,1]
   }
   n20 = Anim Dynamics - Bone: dyn_fr_zip!(ComponentPose=@n19, Alpha=1, GravityOverride:Vector="0,0,-980", ExternalForce="0,0,0") -> (Pose) {
    angularDampingOverride=0.9999
    angularSpringConstant=8000
    bAngularSpring=true
    bOverrideAngularDamping=true
    bUseGravityOverride=true
    gravityScale=0.35
    simulationSpace=RootRelative
    alphaBoolBlend {blendOption=Linear}
    alphaScaleBiasClamp {bClampResult=true}
    boundBone {boneName=dyn_fr_zip}
    componentAppliedLinearAccClamp=<2000 2000 2000>
    componentLinearAccScale=<1 1 1>
    componentLinearVelScale=<1 1 1>
    gravityOverride {z=-980}
    physicsBodyDefinitions [{
      boundBone {boneName=dyn_fr_zip}
      boxExtents=<20 20 20>
      constraintSetup {angularLimitsMax=<0 5 7>, angularLimitsMin=<0 -20 -7>, angularTarget=<1 0 0>}
      localJointOffset=<2 0 0>
     }]
   }
   n21 = Anim Dynamics - Bone: dyn_pants_bk_l!(ComponentPose=@n20, Alpha=1, GravityOverride:Vector="0,0,-980", ExternalForce="0,0,0") -> (Pose) {
    angularDampingOverride=1
    angularSpringConstant=4500
    bAngularSpring=true
    bOverrideAngularDamping=true
    bUseGravityOverride=true
    gravityScale=0.35
    simulationSpace=RootRelative
    alphaBoolBlend {blendOption=Linear}
    boundBone {boneName=dyn_pants_bk_l}
    componentAppliedLinearAccClamp=<2000 2000 2000>
    componentLinearAccScale=<-2 -2 -2>
    componentLinearVelScale=<-2 -2 -2>
    gravityOverride {z=-980}
    physicsBodyDefinitions [{
      boundBone {boneName=dyn_pants_bk_l}
      boxExtents=<10 10 10>
      constraintSetup {angularLimitsMax=<0 20 5>, angularLimitsMin=<0 -15 -5>, angularTarget=<1 0 0>}
      localJointOffset=<1 0 0>
     }]
   }
   n22 = Anim Dynamics - Bone: dyn_pouch_strap_bb!(ComponentPose=@n21, Alpha=1, GravityOverride:Vector="0,0,-980", ExternalForce="0,0,0") -> (Pose) {
    angularDampingOverride=1
    angularSpringConstant=4500
    bAngularSpring=true
    bOverrideAngularDamping=true
    bUseGravityOverride=true
    gravityScale=0.35
    simulationSpace=RootRelative
    alphaBoolBlend {blendOption=Linear}
    boundBone {boneName=dyn_pouch_strap_bb}
    componentAppliedLinearAccClamp=<2000 2000 2000>
    componentLinearAccScale=<-2 -2 -2>
    componentLinearVelScale=<-2 -2 -2>
    gravityOverride {z=-980}
    physicsBodyDefinitions [{
      boundBone {boneName=dyn_pouch_strap_bb}
      boxExtents=<10 10 10>
      constraintSetup {angularLimitsMax=<0 20 5>, angularLimitsMin=<0 -15 -5>, angularTarget=<1 0 0>}
      localJointOffset=<1 0 0>
     }]
   }
   n23 = Anim Dynamics - Bone: dyn_pants_fr_r!(ComponentPose=@n22, Alpha=1, GravityOverride:Vector="0,0,-980", ExternalForce="0,0,0") -> (Pose) {
    angularDampingOverride=1
    angularSpringConstant=4500
    bAngularSpring=true
    bOverrideAngularDamping=true
    bUseGravityOverride=true
    gravityScale=0.35
    simulationSpace=RootRelative
    alphaBoolBlend {blendOption=Linear}
    boundBone {boneName=dyn_pants_fr_r}
    componentAppliedLinearAccClamp=<2000 2000 2000>
    componentLinearAccScale=<-2 -2 -2>
    componentLinearVelScale=<-2 -2 -2>
    gravityOverride {z=-980}
    physicsBodyDefinitions [{
      boundBone {boneName=dyn_pants_fr_r}
      boxExtents=<10 10 10>
      constraintSetup {angularLimitsMax=<0 20 5>, angularLimitsMin=<0 -15 -5>, angularTarget=<1 0 0>}
      localJointOffset=<1 0 0>
     }]
   }
   n24 = Anim Dynamics - Bone: dyn_pants_bk_r!(ComponentPose=@n23, Alpha=1, GravityOverride:Vector="0,0,-980", ExternalForce="0,0,0") -> (Pose) {
    angularDampingOverride=1
    angularSpringConstant=4500
    bAngularSpring=true
    bOverrideAngularDamping=true
    bUseGravityOverride=true
    gravityScale=0.35
    simulationSpace=RootRelative
    alphaBoolBlend {blendOption=Linear}
    boundBone {boneName=dyn_pants_bk_r}
    componentAppliedLinearAccClamp=<2000 2000 2000>
    componentLinearAccScale=<-2 -2 -2>
    componentLinearVelScale=<-2 -2 -2>
    gravityOverride {z=-980}
    physicsBodyDefinitions [{
      boundBone {boneName=dyn_pants_bk_r}
      boxExtents=<10 10 10>
      constraintSetup {angularLimitsMax=<0 20 5>, angularLimitsMin=<0 -15 -5>, angularTarget=<1 0 0>}
      localJointOffset=<1 0 0>
     }]
   }
   n25 = "Transform (Modify) Bone - Bone: dyn_pants_base_r"(ComponentPose=@n24, Alpha=1, Translation="-10,0,0", Rotation="0,0,0", Scale="1,1,1") {
    rotationMode=BMM_Additive
    rotationSpace=BCS_BoneSpace
    translationMode=BMM_Additive
    translationSpace=BCS_BoneSpace
    alphaBoolBlend {blendOption=Linear}
    alphaScaleBiasClamp {bInterpResult=true, interpSpeedDecreasing=5}
    boneToModify {boneName=dyn_pants_base_r}
    translation {x=-10}
   }
   n26 = "Transform (Modify) Bone - Bone: dyn_pants_bk_r"(ComponentPose=@n25, Alpha=1, Translation="18,-5,-5", Rotation="0,0,0", Scale="1,1,1") {
    rotationMode=BMM_Additive
    rotationSpace=BCS_BoneSpace
    translationMode=BMM_Additive
    translationSpace=BCS_BoneSpace
    alphaBoolBlend {blendOption=Linear}
    alphaScaleBiasClamp {bInterpResult=true, interpSpeedDecreasing=5}
    boneToModify {boneName=dyn_pants_bk_r}
    translation=<18 -5 -5>
   }
   n27 = "Transform (Modify) Bone - Bone: dyn_pants_fr_r"(ComponentPose=@n26, Alpha=1, Translation="10,0,3", Rotation="0,0,0", Scale="1,1,1") {
    rotationMode=BMM_Additive
    rotationSpace=BCS_BoneSpace
    translationMode=BMM_Additive
    translationSpace=BCS_BoneSpace
    alphaBoolBlend {blendOption=Linear}
    alphaScaleBiasClamp {bInterpResult=true, interpSpeedDecreasing=5}
    boneToModify {boneName=dyn_pants_fr_r}
    translation {x=10, z=3}
   }
   n28 = "Transform (Modify) Bone - Bone: dyn_pants_bk_l"(ComponentPose=@n27, Alpha=1, Translation="5,0,0", Rotation="0,0,0", Scale="1,1,1") {
    rotationMode=BMM_Additive
    rotationSpace=BCS_BoneSpace
    translationMode=BMM_Additive
    translationSpace=BCS_BoneSpace
    alphaBoolBlend {blendOption=Linear}
    alphaScaleBiasClamp {bInterpResult=true, interpSpeedDecreasing=5}
    boneToModify {boneName=dyn_pants_bk_l}
    translation {x=5}
   }
   n29 = "Transform (Modify) Bone - Bone: dyn_pants_fr_l"(ComponentPose=@n28, Alpha=1, Translation="3,0,0", Rotation="0,0,0", Scale="1,1,1") {
    rotationMode=BMM_Additive
    rotationSpace=BCS_BoneSpace
    translationMode=BMM_Additive
    translationSpace=BCS_BoneSpace
    alphaBoolBlend {blendOption=Linear}
    alphaScaleBiasClamp {bInterpResult=true, interpSpeedDecreasing=5}
    boneToModify {boneName=dyn_pants_fr_l}
    translation {x=3}
   }
   n30 = "Transform (Modify) Bone - Bone: dyn_pants_bk_l"!(ComponentPose=@n29, bAlphaBoolEnabled:bool=True, Translation="19.500,3.208,-8.074", Rotation="-10.615,0.003,-0.001", Scale="1,1,1") -> (Pose) {
    alphaInputType=Bool
    rotationMode=BMM_Additive
    rotationSpace=BCS_BoneSpace
    translationMode=BMM_Additive
    translationSpace=BCS_BoneSpace
    alphaBoolBlend {blendOption=Linear}
    alphaScaleBiasClamp {bInterpResult=true, interpSpeedDecreasing=5}
    boneToModify {boneName=dyn_pants_bk_l}
    rotation=<rot -10.615 0.003 -0.001>
    translation=<19.5 3.208 -8.074>
   }
   n31 = "Transform (Modify) Bone - Bone: dyn_pants_bk_r"!(ComponentPose=@n30, bAlphaBoolEnabled:bool=True, Translation="27.371,3.460,-11.496", Rotation="5.811,-0.000,0.000", Scale="1,1,1") -> (Pose) {
    alphaInputType=Bool
    rotationMode=BMM_Additive
    rotationSpace=BCS_BoneSpace
    translationMode=BMM_Additive
    translationSpace=BCS_BoneSpace
    alphaBoolBlend {blendOption=Linear}
    alphaScaleBiasClamp {bInterpResult=true, interpSpeedDecreasing=5}
    boneToModify {boneName=dyn_pants_bk_r}
    rotation=<rot 5.811 0 0>
    translation=<27.371 3.46 -11.496>
   }
   n32 = "Transform (Modify) Bone - Bone: dyn_pants_bk_r"!(ComponentPose=@n31, Alpha:float=1, Translation="-7,0,0", Rotation="10,0,0", Scale="1,1,1") -> (Pose) {
    rotationMode=BMM_Additive
    rotationSpace=BCS_BoneSpace
    translationMode=BMM_Additive
    translationSpace=BCS_BoneSpace
    alphaBoolBlend {blendOption=Linear}
    alphaScaleBiasClamp {bInterpResult=true, interpSpeedDecreasing=5}
    boneToModify {boneName=dyn_pants_bk_r}
    rotation {pitch=10}
    translation {x=-7}
   }
   n33 = "Transform (Modify) Bone - Bone: dyn_pants_bk_l"!(ComponentPose=@n32, Alpha:float=1, Translation="-10,0,0", Rotation="10,0,0", Scale="1,1,1") -> (Pose) {
    rotationMode=BMM_Additive
    rotationSpace=BCS_BoneSpace
    translationMode=BMM_Additive
    translationSpace=BCS_BoneSpace
    alphaBoolBlend {blendOption=Linear}
    alphaScaleBiasClamp {bInterpResult=true, interpSpeedDecreasing=5}
    boneToModify {boneName=dyn_pants_bk_l}
    rotation {pitch=10}
    translation {x=-10}
   }
   n34 = "Transform (Modify) Bone - Bone: dyn_pouch_b"(ComponentPose=@n33, bAlphaBoolEnabled=True, Translation="0,0,0", Rotation="18,-35,0.498", Scale="1,1,1") {
    alphaInputType=Bool
    rotationMode=BMM_Additive
    rotationSpace=BCS_BoneSpace
    translationMode=BMM_Additive
    translationSpace=BCS_BoneSpace
    alphaBoolBlend {blendOption=Linear}
    alphaScaleBiasClamp {bInterpResult=true, interpSpeedDecreasing=5}
    boneToModify {boneName=dyn_pouch_b}
    rotation=<rot 18 -35 0.498>
   }
   n35 = "Transform (Modify) Bone - Bone: dyn_pouch_strap_aa"!(ComponentPose=@n34, bAlphaBoolEnabled:bool=True, Translation="0,0,0", Rotation="1.651,1.452,0", Scale="1,1,1") -> (Pose) {
    alphaInputType=Bool
    rotationMode=BMM_Additive
    rotationSpace=BCS_BoneSpace
    translationMode=BMM_Additive
    translationSpace=BCS_BoneSpace
    alphaBoolBlend {blendOption=Linear}
    alphaScaleBiasClamp {bInterpResult=true, interpSpeedDecreasing=5}
    boneToModify {boneName=dyn_pouch_strap_aa}
    rotation {pitch=1.651, yaw=1.452}
   }
   n36 = "Transform (Modify) Bone - Bone: dyn_pouch_strap_ab"!(ComponentPose=@n35, bAlphaBoolEnabled:bool=True, Translation="0,0,0", Rotation="-20.263,-6.412,0", Scale="1,1,1") -> (Pose) {
    alphaInputType=Bool
    rotationMode=BMM_Additive
    rotationSpace=BCS_BoneSpace
    translationMode=BMM_Additive
    translationSpace=BCS_BoneSpace
    alphaBoolBlend {blendOption=Linear}
    alphaScaleBiasClamp {bInterpResult=true, interpSpeedDecreasing=5}
    boneToModify {boneName=dyn_pouch_strap_ab}
    rotation {pitch=-20.263, yaw=-6.412}
   }
   n37 = "Transform (Modify) Bone - Bone: dyn_pouch_strap_ba"!(ComponentPose=@n36, bAlphaBoolEnabled:bool=True, Translation="0,0,0", Rotation="1.777,7.566,0", Scale="1,1,1") -> (Pose) {
    alphaInputType=Bool
    rotationMode=BMM_Additive
    rotationSpace=BCS_BoneSpace
    translationMode=BMM_Additive
    translationSpace=BCS_BoneSpace
    alphaBoolBlend {blendOption=Linear}
    alphaScaleBiasClamp {bInterpResult=true, interpSpeedDecreasing=5}
    boneToModify {boneName=dyn_pouch_strap_ba}
    rotation {pitch=1.777, yaw=7.566}
   }
   n38 = "Transform (Modify) Bone - Bone: dyn_pouch_strap_bb"!(ComponentPose=@n37, bAlphaBoolEnabled:bool=True, Translation="0,0,0", Rotation="-13.673,0,0", Scale="1,1,1") -> (Pose) {
    alphaInputType=Bool
    rotationMode=BMM_Additive
    rotationSpace=BCS_BoneSpace
    translationMode=BMM_Additive
    translationSpace=BCS_BoneSpace
    alphaBoolBlend {blendOption=Linear}
    alphaScaleBiasClamp {bInterpResult=true, interpSpeedDecreasing=5}
    boneToModify {boneName=dyn_pouch_strap_bb}
    rotation {pitch=-13.673}
   }
   n39 = "Transform (Modify) Bone - Bone: dyn_pouch_a"(ComponentPose=@n38, Alpha=1, Translation="0,0,0", Rotation="-10,0,0", Scale="1,1,1") {
    rotationMode=BMM_Additive
    translationMode=BMM_Additive
    alphaBoolBlend {blendOption=Linear}
    alphaScaleBiasClamp {bInterpResult=true, interpSpeedIncreasing=2}
    boneToModify {boneName=dyn_pouch_a}
    rotation {pitch=-10}
   }
   n40 = "Transform (Modify) Bone - Bone: dyn_pouch_b"!(ComponentPose=@n39, Alpha:float=1, Translation="0,0,0", Rotation="-20,0,0", Scale="1,1,1") -> (Pose) {
    rotationMode=BMM_Additive
    translationMode=BMM_Additive
    alphaBoolBlend {blendOption=Linear}
    alphaScaleBiasClamp {bInterpResult=true, interpSpeedIncreasing=2}
    boneToModify {boneName=dyn_pouch_b}
    rotation {pitch=-20}
   }
   n41 = "Transform (Modify) Bone - Bone: dyn_pouch_b"(ComponentPose=@n40, Alpha=1, Translation="0,0,0", Rotation="-20,0,0", Scale="1,1,1") {
    rotationMode=BMM_Additive
    translationMode=BMM_Additive
    alphaBoolBlend {blendOption=Linear}
    alphaScaleBiasClamp {bInterpResult=true, interpSpeedIncreasing=2}
    boneToModify {boneName=dyn_pouch_b}
    rotation {pitch=-20}
   }
   n42 = "Transform (Modify) Bone - Bone: dyn_pouch_b"!(ComponentPose=@n41, bAlphaBoolEnabled:bool=Get bIsSkydiveDiveMode(), Translation="0,0,0", Rotation="20,0,0", Scale="1,1,1") -> (Pose) {
    alphaInputType=Bool
    rotationMode=BMM_Additive
    translationMode=BMM_Additive
    alphaBoolBlend {blendOption=Linear}
    alphaScaleBiasClamp {bInterpResult=true, interpSpeedIncreasing=2}
    boneToModify {boneName=dyn_pouch_b}
    rotation {pitch=20}
   }
   n43 = "Transform (Modify) Bone - Bone: dyn_pouch_a_base"(ComponentPose=@n42, Alpha=1, Translation="0,0,0", Rotation="-23.364,-0.060,0.034", Scale="1,1,1") {
    rotationMode=BMM_Additive
    rotationSpace=BCS_BoneSpace
    translationMode=BMM_Additive
    translationSpace=BCS_BoneSpace
    alphaBoolBlend {blendOption=Linear}
    alphaScaleBiasClamp {bInterpResult=true, interpSpeedDecreasing=5, interpSpeedIncreasing=5}
    boneToModify {boneName=dyn_pouch_a_base}
    rotation=<rot -23.364 -0.06 0.034>
   }
   n44 = "Transform (Modify) Bone - Bone: dyn_pouch_b"!(ComponentPose=@n43, Alpha:float=1, Translation="0,0,0", Rotation="-21.843,0.000,-0.000", Scale="1,1,1") -> (Pose) {
    rotationMode=BMM_Additive
    translationMode=BMM_Additive
    alphaBoolBlend {blendOption=Linear}
    alphaScaleBiasClamp {bInterpResult=true, interpSpeedDecreasing=5, interpSpeedIncreasing=5}
    boneToModify {boneName=dyn_pouch_b}
    rotation {pitch=-21.843}
   }
   n45 = "Transform (Modify) Bone - Bone: dyn_pouch_b"!(ComponentPose=@n44, bAlphaBoolEnabled:bool=True, Translation="0,-0.5,0", Rotation="4,0,15", Scale="1,1,1") -> (Pose) {
    alphaInputType=Bool
    rotationMode=BMM_Additive
    translationMode=BMM_Additive
    alphaBoolBlend {blendOption=Linear}
    alphaScaleBiasClamp {bInterpResult=true, interpSpeedIncreasing=2}
    boneToModify {boneName=dyn_pouch_b}
    rotation {pitch=4, roll=15}
    translation {y=-0.5}
   }
   n46 = "Transform (Modify) Bone - Bone: dyn_pouch_b"(ComponentPose=@n45, bAlphaBoolEnabled=True, Translation="-5,5,0", Rotation="7,0,8", Scale="1,1,1") {
    alphaInputType=Bool
    rotationMode=BMM_Additive
    translationMode=BMM_Additive
    alphaBoolBlend {blendOption=Linear}
    alphaScaleBiasClamp {bInterpResult=true, interpSpeedIncreasing=2}
    boneToModify {boneName=dyn_pouch_b}
    rotation {pitch=7, roll=8}
    translation=<-5 5>
   }
   Output Pose(Result=Component To Local(ComponentPose=@n46)) {layerGroup=None}
  ]}
 EventGraph {nodes [
   Event BlueprintUpdateAnimation()
   > n1 = Sequence!(execute) -> (then_0, then_1, then_2:Exec, then_3:Exec, then_4:Exec, then_5:Exec)
   > n2 = Sequence!(execute) -> (then_0, then_1, then_2, then_3)
   n3 = GetDeltaTransformFromRefPose(self=GetOwnerSkeletalMeshComponent(), BoneName=thigh_r, BaseName=pelvis)
   Set uplegRot_R(execute=@n2.then_1, uplegRot_R:double=BreakRotator(InRot=BreakTransform(InTransform=@n3).Rotation).Yaw) -> (Output_Get:double)
   n4 = GetDeltaTransformFromRefPose(self=GetOwnerSkeletalMeshComponent(), BoneName=calf_r, BaseName=thigh_r)
   Set kneeRot_R(execute=@n2.then_3, kneeRot_R:double=BreakRotator(InRot=BreakTransform(InTransform=@n4).Rotation).Yaw) -> (Output_Get:double)
   n5 = GetDeltaTransformFromRefPose(self=GetOwnerSkeletalMeshComponent(), BoneName=calf_l, BaseName=thigh_l)
   Set kneeRot_L(execute=@n2.then_2, kneeRot_L:double=BreakRotator(InRot=BreakTransform(InTransform=@n5).Rotation).Yaw) -> (Output_Get:double)
   n6 = GetDeltaTransformFromRefPose(self=GetOwnerSkeletalMeshComponent(), BoneName=thigh_l, BaseName=pelvis)
   Set uplegRot_L(execute=@n2.then_0, uplegRot_L:double=BreakRotator(InRot=BreakTransform(InTransform=@n6).Rotation).Yaw) -> (Output_Get:double)
   n7 = Branch(execute=@n1.then_1, Condition=Get bIsCrouching())
   > Set isCrouchingRev(isCrouchingRev:double=0) -> (Output_Get:double)
   Set isCrouchingRev(execute=@n7.else, isCrouchingRev=1)
   n8 = "Gravity Override State Function (Parameter Asset) (Nativized)"(
    execute=@n1.then_2
    CharacterPartAnimInstance=Self-Reference()
    GameTimeInSeconds=GetGameTimeInSeconds()
    Parameters=&/M_MED_SummerSuitable_Parameters
   )
   > Set Gravity Override(Gravity Override=@n8)
   n9 = NOT Boolean(A=Get bIsMoving2D())
   n10 = Get bIsCrouching()
   Set crouch not move(execute=@n1.then_3, crouch not move:double="To Float (Boolean)"(InBool=AND Boolean(A=@n10, B=@n9))) -> (Output_Get:double)
   n11 = NOT Boolean(A=Get bIsMoving2D())
   n12 = Get bIsDBNO()
   Set DBNO not moving(execute=@n1.then_4, DBNO not moving:double="To Float (Boolean)"(InBool=AND Boolean(A=@n12, B=@n11))) -> (Output_Get:double)
   > n13 = Branch(Condition=OR Boolean!(
     A=Get bIsSkyDiving()
     B=Get bIsCrouching()
     C:bool=Get bIsPlayingEmote()
     D:bool=Get bIsJumping()
     E:bool=Get bIsFalling()
     F:bool=Get bIsMoving2D()
    ) -> (ReturnValue))
   > Set isIdle(isIdle:double=0) -> (Output_Get:double)
   Set isIdle(execute=@n13.else, isIdle=1)
   n14 = GetOwnerSkeletalMeshComponent()
   n15 = BreakRotator(InRot=BreakTransform(InTransform=GetDeltaTransformFromRefPose(self=@n14, BoneName=thigh_r, BaseName=pelvis)).Rotation)
   n16 = OR Boolean!(A=Get bIsJumping(), B=Get bIsFalling(), C) -> (ReturnValue)
   Set jump_alpha(
    execute=@n1.then_5
    jump_alpha:double=float * float(A=MapRangeClamped(Value=@n15.Yaw, InRangeA=0, InRangeB=-20), B:bool=@n16) {is=title K2Node_PromotableOperator}
   ) -> (Output_Get:double)
   n17 = OR Boolean!(A=Get bIsJumping(), B=Get bIsFalling()) -> (ReturnValue)
   n18 = Get bIsTacticalSprinting()
   > Set isTacticalJump(isTacticalJump:double="To Float (Boolean)"(InBool=AND Boolean(A=@n18, B=@n17))) -> (Output_Get:double)
   AND Boolean(A=@n16, B=Get bIsTacticalSprinting())
  ]}
]
variables [
 crouch not move {type=number}
 DBNO not moving {type=number}
 Gravity Override {type=Vector}
 Gravity Override Params {type=GravityOverrideParamsStruct}
 isCrouchingRev {type=number}
 isIdle {type=number}
 isTacticalJump {type=number}
 jump_alpha {type=number}
 kneeRot_L {type=number}
 kneeRot_R {type=number}
 NewVar {type=number}
 uplegRot_L {type=number}
 uplegRot_R {type=number}
]
