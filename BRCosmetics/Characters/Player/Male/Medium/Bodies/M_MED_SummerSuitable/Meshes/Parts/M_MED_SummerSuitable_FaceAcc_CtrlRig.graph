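# Control Rig for the face accessory part of the M_MED_SummerSuitable male player cosmetic, adding pose-driven corrective offsets to hood bones.
# A Modify Transforms step with an identity offset on upperarm_r runs first, then a Spherical Pose Reader measures that bone's orientation.
# The reader's output weights a second Modify Transforms on right-side hood bones, including dyn_hood_r.
# The pose reader has debug drawing enabled.
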ControlRigBlueprint /BRCosmetics/Characters/Player/Male/Medium/Bodies/M_MED_SummerSuitable/Meshes/Parts/M_MED_SummerSuitable_FaceAcc_CtrlRig
graphs [
 RigVMFunctionLibrary {links [], nodes []}
 RigVMModel {nodes [
   RigUnit_BeginExecution = &RigUnit_BeginExecution()
   &RigUnit_ModifyTransforms(
    ItemToModify=((Item=(Type=Bone,Name="upperarm_r"),Transform=(Rotation=(X=0,Y=0,Z=0,W=1),Translation=(X=0,Y=0,Z=0),Scale3D=(X=1,Y=1,Z=1))))
    Weight=1
    WeightMinimum=0
    ExecutePin=@RigUnit_BeginExecution.ExecutePin
   )
   > n1 = &RigUnit_SphericalPoseReader(
    ActiveRegionScaleFactors=(PositiveWidth=1,NegativeWidth=1,PositiveHeight=1,NegativeHeight=1)
    ActiveRegionSize=0.05
    Debug=(bDrawDebug=true,bDraw2D=False,bDrawLocalAxes=False,DebugScale=25,DebugSegments=20,DebugThickness=0.25)
    DriverAxis=(X=-1,Y=0,Z=0)
    DriverItem=(Type=Bone,Name="upperarm_r")
    FalloffRegionScaleFactors=(PositiveWidth=0,NegativeWidth=1,PositiveHeight=0.3,NegativeHeight=0.3)
    FalloffSize=0.3
    FlipHeightScaling=False
    FlipWidthScaling=False
    OptionalParentItem=(Type=Bone,Name="None")
    RotationOffset=(X=-105,Y=0,Z=55)
   )
   > &RigUnit_ModifyTransforms(
    ItemToModify="((Item=(Type=Bone,Name=\"dyn_hood_r\"),Transform=(Rotation=(X=-0.0282,Y=0.2587,Z=0.0076,W=0.9655),Translation=(X=0.9034,Y=-0.0881,Z=-1.3207),Scale3D=(X=1,Y=1,Z=1))),(Item=(Type=Bone,Name=\"dyn_hood_bottom_base_r\"),Transform=(Rotation=(X=0,Y=0,Z=0,W=1),Transla ...(truncated)"
    WeightMinimum=0
    Weight=@n1.OutputParam
   )
  ]}
]
hierarchy [
 root {loc=<0 0 0>, rot=<0 0 0 1>}
 pelvis {parent=root, loc=<0 -0.7071 93.1191>, rot=<0.0127 0.707 -0.0127 -0.707>}
 spine_01 {parent=pelvis, loc=<6.7494 -0.6388 0>, rot=<0 0 0.034 -0.9994>}
 spine_02 {parent=spine_01, loc=<7.2518 -0.035 0>, rot=<0 0 0.0238 -0.9997>}
 spine_03 {parent=spine_02, loc=<6.0104 0 0>, rot=<0 0 0.0149 -0.9999>}
 spine_04 {parent=spine_03, loc=<8.1024 0 0>, rot=<0 0 -0.0132 -0.9999>}
 spine_05 {parent=spine_04, loc=<10.0146 0 0>, rot=<0 0 -0.0015 -1>}
 neck_01 {parent=spine_05, loc=<11.3138 0.1529 0>, rot=<0 0 -0.2003 -0.9797>}
 neck_02 {parent=neck_01, loc=<4.5627 0 0>, rot=<0 0 0.0504 -0.9987>}
 head {parent=neck_02, loc=<5.3914 0 0>, rot=<0 0 0.115 -0.9934>}
 dyn_hood_base_l {parent=head, loc=<-1.4835 -0.9248 -10.111>, rot=<0.0041 -1 0 0>}
 dyn_hood_l {parent=dyn_hood_base_l, loc=<0 0 0>, rot=<0 0 0 1>}
 dyn_hood_base_r {parent=head, loc=<-1.4836 -0.9248 10.111>, rot=<0 0 1 -0.0041>}
 dyn_hood_r {parent=dyn_hood_base_r, loc=<0 0 0>, rot=<0 0 0 1>}
 dyn_hood_01_r {parent=dyn_hood_r, loc=<0 0 0>, rot=<0 0 0 1>}
 clavicle_l {parent=spine_05, loc=<4.3575 -0.1466 -2.3223>, rot=<0.0465 -0.7289 -0.009 -0.6829>}
 upperarm_l {parent=clavicle_l, loc=<15.5111 0 0>, rot=<0.0226 -0.3234 -0.0088 -0.9459>}
 lowerarm_l {parent=upperarm_l, loc=<26.6514 0 0>, rot=<0.0391 -0.0099 -0.2453 -0.9686>}
 clavicle_r {parent=spine_05, loc=<4.3575 -0.1466 2.3223>, rot=<0.7289 0.0465 -0.6829 0.009>}
 upperarm_r {parent=clavicle_r, loc=<-15.5111 0 0>, rot=<0.0226 -0.3234 -0.0088 -0.9459>}
 lowerarm_r {parent=upperarm_r, loc=<-26.6514 0 0>, rot=<0.0391 -0.0099 -0.2453 -0.9686>}
 dyn_hood_base {parent=spine_05, loc=<14.815 -7.6565 0>, rot=<0.0401 -0.9992 0 0>}
 dyn_hood {parent=dyn_hood_base, loc=<0 0 0>, rot=<0 0 0 1>}
 dyn_hood_bottom_base_r {parent=spine_05, loc=<14.5807 3.7969 8.7615>, rot=<0 0 0.9992 -0.0401>}
 dyn_hood_bottom_r {parent=dyn_hood_bottom_base_r, loc=<0 0 0>, rot=<0 0 0 1>}
]
