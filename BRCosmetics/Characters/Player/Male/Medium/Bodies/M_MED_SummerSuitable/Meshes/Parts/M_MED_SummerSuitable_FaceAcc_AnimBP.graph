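# Animation Blueprint for the SummerSuitable male face accessory, driving its hood bones to follow the character.
# It copies head and body poses, blending body in from the upper arms, then runs a Control Rig.
# Additive bone offsets adjust the right hood, weighted by right-arm pose measured against the reference pose, with further offsets while crouching.
# Spring-based AnimDynamics on left and right hood bones add secondary motion last.
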AnimBlueprint /BRCosmetics/Characters/Player/Male/Medium/Bodies/M_MED_SummerSuitable/Meshes/Parts/M_MED_SummerSuitable_FaceAcc_AnimBP
target_skeleton=&/M_MED_SummerSuitable_FaceAcc_Skeleton
graphs [
 AnimGraph {nodes [
   n1 = Get Body_Mesh()
   n2 = Get Head_Mesh() -> (Head_Mesh:Skeletal Mesh Component Object Reference)
   n3 = Control Rig(Source=Layered blend per bone(BasePose=Copy Pose From Mesh(SourceMeshComponent=@n2) {bUseAttachedParent=true}, BlendPoses_0=Copy Pose From Mesh(SourceMeshComponent=@n1)) {layerSetup [{branchFilters [{boneName=upperarm_l}, {boneName=upperarm_r}]}]}) {
    alphaBoolBlend {blendOption=Linear}
    controlRigAssetReference {blueprintRigClass=&/M_MED_SummerSuitable_FaceAcc_CtrlRig.M_MED_SummerSuitable_FaceAcc_CtrlRig_C}
   }
   n4 = Get hoodFix_alpha_r() -> (hoodFix_alpha_r:double)
   n5 = "Transform (Modify) Bone - Bone: dyn_hood_bottom_base_r"(ComponentPose=Local To Component(LocalPose=@n3), Alpha=@n4, Translation="0,0,-2", Rotation="0,0,-60", Scale="1,1,1") {
    rotationMode=BMM_Additive
    rotationSpace=BCS_BoneSpace
    translationMode=BMM_Additive
    translationSpace=BCS_BoneSpace
    alphaBoolBlend {blendOption=Linear}
    boneToModify {boneName=dyn_hood_bottom_base_r}
    rotation {roll=-60}
    translation {z=-2}
   }
   n6 = "Transform (Modify) Bone - Bone: dyn_hood_base_r"(ComponentPose=@n5, Alpha=@n4, Translation="0,0,0", Rotation="0,0,-25", Scale="1,1,1") {
    rotationMode=BMM_Additive
    rotationSpace=BCS_BoneSpace
    translationMode=BMM_Additive
    translationSpace=BCS_BoneSpace
    alphaBoolBlend {blendOption=Linear}
    boneToModify {boneName=dyn_hood_base_r}
    rotation {roll=-25}
   }
   n7 = Get bIsCrouching()
   n8 = "Transform (Modify) Bone - Bone: dyn_hood_bottom_base_r"!(ComponentPose=@n6, bAlphaBoolEnabled:bool=@n7, Translation="0,0,-2.5", Rotation="0,0,-50", Scale="1,1,1") -> (Pose) {
    alphaInputType=Bool
    rotationMode=BMM_Additive
    rotationSpace=BCS_BoneSpace
    translationMode=BMM_Additive
    translationSpace=BCS_BoneSpace
    alphaBoolBlend {blendOption=Linear}
    boneToModify {boneName=dyn_hood_bottom_base_r}
    rotation {roll=-50}
    translation {z=-2.5}
   }
   n9 = "Transform (Modify) Bone - Bone: dyn_hood_base_r"!(ComponentPose=@n8, bAlphaBoolEnabled:bool=@n7, Translation="0,0,0", Rotation="0,0,-20", Scale="1,1,1") -> (Pose) {
    alphaInputType=Bool
    rotationMode=BMM_Additive
    rotationSpace=BCS_BoneSpace
    translationMode=BMM_Additive
    translationSpace=BCS_BoneSpace
    alphaBoolBlend {blendOption=Linear}
    boneToModify {boneName=dyn_hood_base_r}
    rotation {roll=-20}
   }
   n10 = Anim Dynamics - Bone: dyn_hood_l(ComponentPose=@n9, Alpha=1, ExternalForce="0,0,0") {
    angularDampingOverride=1
    angularSpringConstant=1500
    bAngularSpring=true
    bGravityOverrideInSimSpace=true
    bOverrideAngularDamping=true
    bUseGravityOverride=true
    gravityScale=0.35
    simulationSpace=RootRelative
    alphaBoolBlend {blendOption=Linear}
    boundBone {boneName=dyn_hood_l}
    componentAppliedLinearAccClamp=<3000 3000 3000>
    gravityOverride {z=-980}
    physicsBodyDefinitions [{
      sphereCollisionRadius=0
      boundBone {boneName=dyn_hood_l}
      boxExtents=<10 20 20>
      constraintSetup {angularLimitsMax=<15 5 5>, angularLimitsMin=<-15 -5 -5>, angularTarget=<0 0 0>}
      localJointOffset=<10 0 0>
     }]
   }
   n11 = Anim Dynamics - Bone: dyn_hood_r(ComponentPose=@n10, Alpha=1, ExternalForce="0,0,0") {
    angularDampingOverride=1
    angularSpringConstant=1500
    bAngularSpring=true
    bGravityOverrideInSimSpace=true
    bOverrideAngularDamping=true
    bUseGravityOverride=true
    gravityScale=0.35
    simulationSpace=RootRelative
    alphaBoolBlend {blendOption=Linear}
    boundBone {boneName=dyn_hood_r}
    componentAppliedLinearAccClamp=<3000 3000 3000>
    gravityOverride {z=-980}
    physicsBodyDefinitions [{
      sphereCollisionRadius=0
      boundBone {boneName=dyn_hood_r}
      boxExtents=<10 20 20>
      constraintSetup {angularLimitsMax=<15 5 5>, angularLimitsMin=<-15 -5 -5>, angularTarget=<0 0 0>}
      localJointOffset=<10 0 0>
     }]
   }
   Output Pose(Result=Component To Local(ComponentPose=@n11)) {layerGroup=None}
  ]}
 EventGraph {nodes [
   Event BlueprintUpdateAnimation()
   > n1 = Sequence!(execute) -> (then_0, then_1, then_2:Exec, then_3:Exec)
   n2 = GetOwnerSkeletalMeshComponent()
   n3 = BreakRotator(InRot=BreakTransform(InTransform=GetDeltaTransformFromRefPose(self=@n2, BoneName=upperarm_r, BaseName=clavicle_r)).Rotation)
   > Set armForward_R(armForward_R:double=@n3.Yaw) -> (Output_Get:double)
   > Set armUp_R(armUp_R:double=@n3.Pitch) -> (Output_Get:double)
   n4 = Branch(execute=@n1.then_1, Condition="float > float"(A=Get armForward_R() -> (armForward_R:double), B=90) {is=title K2Node_CallFunction})
   > Set hoodFix_alpha_r(hoodFix_alpha_r:double=MapRangeClamped(Value=Get armUp_R() -> (armUp_R:double), InRangeA=-10, InRangeB=10)) -> (Output_Get:double)
   Set hoodFix_alpha_r(execute=@n4.else, hoodFix_alpha_r=0)
   n5 = Event OnCharacterCustomizationLoaded()
   Set Body_Mesh(execute=[@n5.then, @n8.then], Body_Mesh=Get Array Item with EFort Custom Part Type(Array=@n5.CustomMeshParts))
   > n6 = Cast To PlayerPawn_Athena(Object=TryGetPawnOwner())
   > n7 = Set Pawn Owner(Pawn Owner:Player Pawn Athena Object Reference=@n6) -> (Output_Get:Player Pawn Athena Object Reference)
   > Set Head_Mesh(Head_Mesh:Skeletal Mesh Component Object Reference=GetSkeletalMeshForPartType(self=@n7, PartType=Head)) -> (Output_Get:Skeletal Mesh Component Object Reference)
   n8 = Event BlueprintInitializeAnimation()
   GetOwnerSkeletalMeshComponent()
  ]}
]
variables [
 armForward_R {type=number}
 armUp_R {type=number}
 Body_Mesh {type=SkeletalMeshComponent}
 hair_rotation {type=Rotator}
 Head_Mesh {type=SkeletalMeshComponent}
 hoodFix_alpha_r {type=number}
 Pawn Owner {type=PlayerPawn_Athena_C}
]
